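# Translucent UI material for The Block on the minimap, fading a block texture according to height.
# It subtracts the Height value in the MPC_Minimap parameter collection from a HeightPosition parameter, then divides by SmoothTransition and saturates.
# That factor scales the MinimapBlock texture's alpha, so it fades to zero as Height rises to HeightPosition.
# The texture parameter defaults to solid black.
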Material /Figment_S09/Athena/HUD/MiniMap/M_MinimapTheBlock
graph {
 properties {blendMode=BLEND_Translucent, materialDomain=MD_UI, pixelDepthOffsetMode=PDOM_Legacy, usedShadingModels=Default Lit}
 nodes [
  n1 = Subtract(
   A=ScalarParameter() {defaultValue=768, parameterName=HeightPosition}
   B=CollectionParameter() {collection=/Game/Athena/HUD/MiniMap/MPC_Minimap, parameterName=Height}
  )
  n2 = ScalarParameter() {defaultValue=300, parameterName=SmoothTransition}
  Multiply(
   A=TextureSampleParameter2D() {parameterName=MinimapBlock, texture=/Figment_S09/UI/CodeReferenced/SolidBlack}.A
   B=Saturate(None=Divide(A=@n1, B=@n2))
  )
 ]
}
